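# Physics asset for the RankedRube two-handed pickaxe, giving capsule bodies to the handle and to a dynamic bow with two rope tails.
# The bow master rotates freely relative to the handle, and the bow base is rotationally locked to that master.
# The left and right bow pieces, a three-bone left rope and a two-bone right rope branch from the base through angular-limit constraints, with twist locked on the right rope.
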PhysicsAsset /BRCosmetics/Weapons/FORT_Melee/Pickaxe_RankedRube/Meshes/RankedRube_Axe_Physics
bodies {
 Melee_TwoHanded_Handle [{shapeType=Capsule, length=105.4975, radius=2.1614, center=<0.7497 -1.0517 46.7081>, rotation=<rot -0.0296 -90 0.0298>}]
 dyn_bow_MSTR [{shapeType=Capsule, length=0.505, radius=0.505, shapeName=dyn_bow_MSTR_capsule, rotation=<rot 0 -90 0>}]
 dyn_bow_base [{shapeType=Capsule, length=1.6765, radius=1.7515, shapeName=dyn_bow_base_capsule, rotation=<rot 0 0 -90>}]
 dyn_bow_l [{
   shapeType=Capsule
   length=5.8811
   radius=2.4018
   shapeName=dyn_bow_l_capsule
   center=<-4.9341 -0.5506 0.6267>
   rotation=<rot -5.4179 -88.2477 -80.545>
  }]
 dyn_bow_r [{
   shapeType=Capsule
   length=5.679
   radius=2.9448
   shapeName=dyn_bow_r_capsule
   center=<4.7189 0.6075 -0.2468>
   rotation=<rot -0.1945 82.1458 -88.8967>
  }]
 dyn_rope_l_01 [{shapeType=Capsule, length=2.0198, radius=2.4802, shapeName=dyn_rope_l_01_capsule, center=<1.9448 0 0>, rotation=<rot 90 0 0>}]
 dyn_rope_l_02 [{shapeType=Capsule, length=1.2229, radius=3.1219, shapeName=dyn_rope_l_02_capsule, center=<1.9448 0 0>, rotation=<rot 90 0 0>}]
 dyn_rope_l_03 [{shapeType=Capsule, length=3.2699, radius=3.9339, shapeName=dyn_rope_l_03_capsule, center=<2.1318 0 0>, rotation=<rot 0 -90 -90>}]
 dyn_rope_r_01 [{shapeType=Capsule, length=3.6025, radius=1.5059, shapeName=dyn_rope_r_01_capsule, center=<0 -0.3264 0.6399>, rotation=<rot -90 0 0>}]
 dyn_rope_r_02 [{
   shapeType=Capsule
   length=9.9189
   radius=1.6159
   shapeName=dyn_rope_r_02_capsule
   center=<0.2279 0 0.6399>
   rotation=<rot -90 162.6319 -152.6319>
  }]
}
constraints [
 {
  bone1Name=dyn_bow_MSTR
  bone2Name=Melee_TwoHanded_Handle
  swing1LimitDegrees=45
  swing1Motion=Free
  swing2LimitDegrees=45
  swing2Motion=Free
  twistLimitDegrees=45
  twistMotion=Free
 }
 {
  bone1Name=dyn_bow_base
  bone2Name=dyn_bow_MSTR
  swing1LimitDegrees=45
  swing1Motion=Locked
  swing2LimitDegrees=20
  swing2Motion=Locked
  twistLimitDegrees=1
  twistMotion=Locked
 }
 {bone1Name=dyn_rope_l_01, bone2Name=dyn_bow_base, swing1LimitDegrees=20, swing2LimitDegrees=10, twistLimitDegrees=1}
 {bone1Name=dyn_rope_l_02, bone2Name=dyn_rope_l_01, swing1LimitDegrees=50, swing2LimitDegrees=30, twistLimitDegrees=1}
 {bone1Name=dyn_rope_l_03, bone2Name=dyn_rope_l_02, swing1LimitDegrees=25, swing2LimitDegrees=25, twistLimitDegrees=15}
 {bone1Name=dyn_rope_r_01, bone2Name=dyn_bow_base, swing1LimitDegrees=60, swing2LimitDegrees=10, twistLimitDegrees=1, twistMotion=Locked}
 {bone1Name=dyn_rope_r_02, bone2Name=dyn_rope_r_01, swing1LimitDegrees=60, swing2LimitDegrees=20, twistLimitDegrees=1, twistMotion=Locked}
 {bone1Name=dyn_bow_l, bone2Name=dyn_bow_base, swing1LimitDegrees=45, swing2LimitDegrees=45, twistLimitDegrees=0}
 {bone1Name=dyn_bow_r, bone2Name=dyn_bow_base, swing1LimitDegrees=45, swing2LimitDegrees=45, twistLimitDegrees=0}
]
